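; Low-altitude cloud actor for the Helios environment that configures a cloud mesh, its hidden shadow casting and an accompanying post-process volume.
; SetParameters runs at construction and BeginPlay, choosing the mesh by complexity and translucent backplate option.
; It packs shape, shading, texture and color settings into Custom Primitive Data, with a helper splitting vectors across three float slots.
; The post-process mesh also receives box extent, actor axes and fade this way.
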(event EventBeginPlay
  (CallFunction|SetParameters))

(fn SetFloatCPDsFromVector
  (params (Static%20Mesh :type Static%20Mesh%20Component%20Object%20Reference) (DataIndex :type Integer) (Value :type Vector))
  (bind (_value_x _value_y _value_z) (Math|Vector|BreakVector (Variables|GetValue_X)))
  (bind _value_x_2 (Math|Vector|BreakVector (Variables|GetValue_X)))
  (bind (_value_x_3 _value_y_2) (Math|Vector|BreakVector (Variables|GetValue_X)))
  (Rendering|Material|SetDefaultCustomPrimitiveDataFloat (Variables|GetStatic%20Mesh) (Variables|GetDataIndex) _value_x_2)
  (Rendering|Material|SetDefaultCustomPrimitiveDataFloat (Variables|GetStatic%20Mesh) (+ (Variables|GetDataIndex) 1) _value_y_2)
  (Rendering|Material|SetDefaultCustomPrimitiveDataFloat (Variables|GetStatic%20Mesh) (+ (Variables|GetDataIndex) 2) _value_z))

(fn SetParameters
  (bind _postprocessarea (Variables|Default|GetPostProcessArea))
  (bind _cloudmesh (Variables|Default|GetCloudMesh))
  (bind _postprocessmesh (Variables|Default|GetPostProcessMesh))
  (Utilities|FlowControl|Sequence
    (:then_0
      (Components|StaticMesh|SetStaticMesh _cloudmesh (Utilities|Select :Index (Variables|Parameters|GetMeshComplexity) :NewEnumerator0 (select (Variables|Parameters|GetTranslucentBackplate) "/Game/Environments/Helios/TechArt/LowAltitudeClouds/V2/SM_LowAltitudeClouds_V2_LowTrans.SM_LowAltitudeClouds_V2_LowTrans" "/Game/Environments/Helios/TechArt/LowAltitudeClouds/V2/SM_LowAltitudeClouds_V2_Low.SM_LowAltitudeClouds_V2_Low") :NewEnumerator1 (select (Variables|Parameters|GetTranslucentBackplate) "/Game/Environments/Helios/TechArt/LowAltitudeClouds/V2/SM_LowAltitudeClouds_V2_MediumTrans.SM_LowAltitudeClouds_V2_MediumTrans" "/Game/Environments/Helios/TechArt/LowAltitudeClouds/V2/SM_LowAltitudeClouds_V2_Medium.SM_LowAltitudeClouds_V2_Medium")))
      (Rendering|Material|SetDefaultCustomPrimitiveDataFloat _cloudmesh 0 (Variables|Parameters|GetMeshSeed))
      (CallFunction|SetFloatCPDsfromVector _cloudmesh 1 (Variables|Parameters|GetSizeMax))
      (CallFunction|SetFloatCPDsfromVector _cloudmesh 4 (Variables|Parameters|GetSizeMin))
      (Rendering|Material|SetDefaultCustomPrimitiveDataFloat _cloudmesh 7 (Variables|Parameters|GetTrapezoidalShape))
      (CallFunction|SetFloatCPDsfromVector _cloudmesh 8 (* (Variables|Parameters|GetShadingScale) "0.700000,0.700000,0.700000"))
      (CallFunction|SetFloatCPDsfromVector _cloudmesh 11 (* (Variables|Parameters|GetShadingOffset) "0.010000,0.010000,0.010000"))
      (CallFunction|SetFloatCPDsfromVector _cloudmesh 14 (* (Variables|Parameters|GetTransmitanceScale) "0.400000,0.400000,0.400000"))
      (CallFunction|SetFloatCPDsfromVector _cloudmesh 17 (* (Variables|Parameters|GetTransmitanceOffset) "0.010000,0.010000,0.010000"))
      (Rendering|Material|SetDefaultCustomPrimitiveDataFloat _cloudmesh 20 (Variables|Parameters|GetFade))
      (Rendering|Material|SetDefaultCustomPrimitiveDataFloat _cloudmesh 21 (Variables|Parameters|GetAlphaScale))
      (Rendering|Material|SetDefaultCustomPrimitiveDataFloat _cloudmesh 22 (Variables|Parameters|GetTextureScale))
      (Rendering|Material|SetDefaultCustomPrimitiveDataFloat _cloudmesh 23 (Variables|Parameters|GetTextureSeed))
      (Rendering|Material|SetDefaultCustomPrimitiveDataFloat _cloudmesh 24 (Variables|Parameters|GetNormalIntensity))
      (CallFunction|SetFloatCPDsfromVector _cloudmesh 25 (Math|Conversions|ToVector(LinearColor) (Variables|Parameters|GetColorMultiplier)))
      (Rendering|Material|SetDefaultCustomPrimitiveDataFloat _cloudmesh 28 (Variables|Parameters|GetAlphaScaleMin))
      (Rendering|Material|SetDefaultCustomPrimitiveDataFloat _cloudmesh 29 (Variables|Parameters|GetArcShape))
      (Rendering|Material|SetDefaultCustomPrimitiveDataFloat _cloudmesh 30 1.0)
      (Rendering|SetCastHiddenShadow (Variables|Default|GetCloudMeshShadow) (Variables|Parameters|GetCastShadows)))
    (:then_1
      (Rendering|Material|SetMaterial _postprocessarea 0 (Variables|CloudMaterials|GetPostProcessMaterial))
      (Transformation|SetRelativeLocation _postprocessarea (Variables|Parameters|GetPostProcessOffset) false false)
      (Transformation|SetRelativeScale3D _postprocessarea (Variables|Parameters|GetPostProcessScale))
      (CallFunction|SetFloatCPDsfromVector _postprocessmesh 0 (Components|Box|GetScaledBoxExtent (Variables|Default|GetPostProcessArea)))
      (CallFunction|SetFloatCPDsfromVector _postprocessmesh 3 (Math|Vector|GetForwardVector (Transformation|GetActorRotation)))
      (CallFunction|SetFloatCPDsfromVector _postprocessmesh 6 (Math|Vector|GetRightVector (Transformation|GetActorRotation)))
      (CallFunction|SetFloatCPDsfromVector _postprocessmesh 9 (Math|Vector|GetUpVector (Transformation|GetActorRotation)))
      (CallFunction|SetFloatCPDsfromVector _postprocessmesh 12 (* (Variables|Parameters|GetShadingScale) "0.700000,0.700000,0.700000"))
      (CallFunction|SetFloatCPDsfromVector _postprocessmesh 15 (* (Variables|Parameters|GetShadingOffset) "0.010000,0.010000,0.010000"))
      (Rendering|Material|SetDefaultCustomPrimitiveDataFloat _postprocessmesh 18 (Variables|Parameters|GetPostProcessFade))
      (CallFunction|SetFloatCPDsfromVector _postprocessmesh 19 (Math|Conversions|ToVector(LinearColor) (Variables|Parameters|GetColorMultiplier))))))

(fn ConstructionScript
  (CallFunction|SetParameters))
