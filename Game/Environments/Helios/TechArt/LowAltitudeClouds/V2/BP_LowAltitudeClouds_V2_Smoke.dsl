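; Smoke variant Blueprint of the Helios V2 low-altitude clouds, built on a parent class that supplies its base setup.
; BeginPlay runs the parent logic, then binds a listener to the Fortnite player controller's view-target-changed event.
; On each change, the listener sets the PostProcessFade parameter to a fixed value, ignoring which targets are involved.
; Construction only calls the parent, and the overlap and tick events are empty.
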(event EventBeginPlay
  (|CallParentFunction)
  (EventDispatchers|BindEventtoOnViewTargetChanged (Fortnite|GetFortPlayerControllerfromActor 0) (delegate Custom|Listener)))

(event Collision|EventActorBeginOverlap
  (params (OtherActor :type Actor%20Object%20Reference)))

(event Custom|Listener
  (params (PlayerController :type Fort%20Player%20Controller%20Object%20Reference) (OldViewTarget :type Actor%20Object%20Reference) (NewViewTarget :type Actor%20Object%20Reference))
  (Variables|Parameters|SetPostProcessFade 1.5e+03))

(event EventTick
  (params (DeltaSeconds :type Float%20%28single-precision%29)))

(fn ConstructionScript
  (|CallParentFunction))
