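; Low-altitude cloud whisps Blueprint for the Helios environment that configures a cloud mesh and its material inputs.
; SetParameters runs on construction and BeginPlay, choosing the low or medium mesh by MeshComplexity.
; It writes seed, size, shape, fade, opacity, scroll and time-of-day emissive settings into default custom primitive data.
; A helper packs vectors into three consecutive slots; ColorMultiplier overwrites the second OpacityErosion write at slot 23.
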(event EventBeginPlay
  (CallFunction|SetParameters))

(fn SetFloatCPDsFromVector
  (params (Static%20Mesh :type Static%20Mesh%20Component%20Object%20Reference) (DataIndex :type Integer) (Value :type Vector))
  (bind (_value_x _value_y _value_z) (Math|Vector|BreakVector (Variables|GetValue_X)))
  (bind _value_x_2 (Math|Vector|BreakVector (Variables|GetValue_X)))
  (bind (_value_x_3 _value_y_2) (Math|Vector|BreakVector (Variables|GetValue_X)))
  (Rendering|Material|SetDefaultCustomPrimitiveDataFloat (Variables|GetStatic%20Mesh) (Variables|GetDataIndex) _value_x_2)
  (Rendering|Material|SetDefaultCustomPrimitiveDataFloat (Variables|GetStatic%20Mesh) (+ (Variables|GetDataIndex) 1) _value_y_2)
  (Rendering|Material|SetDefaultCustomPrimitiveDataFloat (Variables|GetStatic%20Mesh) (+ (Variables|GetDataIndex) 2) _value_z))

(fn SetParameters
  (bind _cloudmesh (Variables|Default|GetCloudMesh))
  (Components|StaticMesh|SetStaticMesh _cloudmesh (Utilities|Select :Index (Variables|Parameters|GetMeshComplexity) :NewEnumerator0 "/Game/Environments/Helios/TechArt/LowAltitudeClouds/V2/SM_LowAltitudeClouds_V2_Whisps_Low.SM_LowAltitudeClouds_V2_Whisps_Low" :NewEnumerator1 "/Game/Environments/Helios/TechArt/LowAltitudeClouds/V2/SM_LowAltitudeClouds_V2_Whisps_Medium.SM_LowAltitudeClouds_V2_Whisps_Medium"))
  (Rendering|Material|SetDefaultCustomPrimitiveDataFloat _cloudmesh 0 (Variables|Parameters|GetMeshSeed))
  (CallFunction|SetFloatCPDsfromVector _cloudmesh 1 (Variables|Parameters|GetSizeMax))
  (CallFunction|SetFloatCPDsfromVector _cloudmesh 4 (Variables|Parameters|GetSizeMin))
  (Rendering|Material|SetDefaultCustomPrimitiveDataFloat _cloudmesh 7 (Variables|Parameters|GetTrapezoidalShape))
  (Rendering|Material|SetDefaultCustomPrimitiveDataFloat _cloudmesh 20 (Variables|Parameters|GetOpacity))
  (Rendering|Material|SetDefaultCustomPrimitiveDataFloat _cloudmesh 19 (Variables|Parameters|GetOpacityErosion))
  (Rendering|Material|SetDefaultCustomPrimitiveDataFloat _cloudmesh 18 (Variables|Parameters|GetNearFadeStart))
  (Rendering|Material|SetDefaultCustomPrimitiveDataFloat _cloudmesh 17 (Variables|Parameters|GetNearFadeEnd))
  (Rendering|Material|SetDefaultCustomPrimitiveDataFloat _cloudmesh 21 (Variables|Parameters|GetAlphaScale))
  (Rendering|Material|SetDefaultCustomPrimitiveDataFloat _cloudmesh 14 (Variables|Parameters|GetScroll%20Speed))
  (Rendering|Material|SetDefaultCustomPrimitiveDataFloat _cloudmesh 22 (Variables|Parameters|GetAlpha%20Scale%20Z))
  (Rendering|Material|SetDefaultCustomPrimitiveDataFloat _cloudmesh 23 (Variables|Parameters|GetOpacityErosion))
  (Rendering|Material|SetDefaultCustomPrimitiveDataFloat _cloudmesh 16 (Variables|Parameters|GetContactFade))
  (CallFunction|SetFloatCPDsfromVector _cloudmesh 23 (Math|Conversions|ToVector(LinearColor) (Variables|Parameters|GetColorMultiplier)))
  (Rendering|Material|SetDefaultCustomPrimitiveDataFloat _cloudmesh 15 (Variables|Parameters|GetArcShape))
  (CallFunction|SetFloatCPDsfromVector _cloudmesh 8 (Math|Conversions|ToVector(LinearColor) (Variables|Parameters|GetEmissiveMorning)))
  (CallFunction|SetFloatCPDsfromVector _cloudmesh 11 (Math|Conversions|ToVector(LinearColor) (Variables|Parameters|GetEmissiveDay)))
  (CallFunction|SetFloatCPDsfromVector _cloudmesh 26 (Math|Conversions|ToVector(LinearColor) (Variables|Parameters|GetEmissiveEvening)))
  (CallFunction|SetFloatCPDsfromVector _cloudmesh 29 (Math|Conversions|ToVector(LinearColor) (Variables|Parameters|GetEmissiveNight))))

(fn ConstructionScript
  (CallFunction|SetParameters))
