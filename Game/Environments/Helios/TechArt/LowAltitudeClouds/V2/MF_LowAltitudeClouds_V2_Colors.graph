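# Material function that produces the color for the V2 low-altitude clouds in the Helios environment.
# Shadow, light and halo colors come from the MPC_LowAltitudeClouds_V2 parameter collection, keeping only RGB.
# The DotMask input blends shadow to light, then a boosted Halo Mask blends that result toward the halo color.
# The halo blend factor is not clamped, and the final color is doubled before output.
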MaterialFunction /Game/Environments/Helios/TechArt/LowAltitudeClouds/V2/MF_LowAltitudeClouds_V2_Colors
graph {nodes [
  n1 = LinearInterpolate(
   A=ComponentMask(None=CollectionParameter() {collection=&/MPC_LowAltitudeClouds_V2, parameterName=ColorShadow}) {b=true, g=true, r=true}
   B=ComponentMask(None=CollectionParameter() {collection=&MPC_LowAltitudeClouds_V2, parameterName=ColorLight}) {b=true, g=true, r=true}
   Alpha=FunctionInput() {bUsePreviewValueAsDefault=true, inputName=DotMask, inputType=FunctionInput_Scalar}
  )
  n2 = ComponentMask(None=CollectionParameter() {collection=&MPC_LowAltitudeClouds_V2, parameterName=ColorHalo}) {b=true, g=true, r=true}
  n3 = Multiply(A=FunctionInput() {bUsePreviewValueAsDefault=true, inputName=Halo Mask, inputType=FunctionInput_Scalar, sortPriority=1}) {constB=4}
  FunctionOutput(None=Multiply(A=LinearInterpolate(A=@n1, B=@n2, Alpha=@n3)) {constB=2})
 ]}
