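# Translucent unlit overlay material that imitates a post-process fog effect while the camera is inside a low-altitude cloud volume.
# An oriented box test from custom primitive data fades opacity near the edges and yields a large downward offset outside, apparently to hide the mesh.
# Sliced 3D noise, sampled by view direction and camera position, modulates opacity and sky-light shading.
# A shared color function adds tinting, scaled per primitive.
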Material /Game/Environments/Helios/TechArt/LowAltitudeClouds/V2/M_LowAltitudeClouds_PostProcess_V2
graph {
 properties {
  bDisableDepthTest=true
  bUseTranslucencyVertexFog=false
  bUsedWithWater=true
  blendMode=BLEND_Translucent
  blendableLocation=BL_SceneColorAfterDOF
  pixelDepthOffsetMode=PDOM_Legacy
  shadingModel=MSM_Unlit
  usedShadingModels=Unlit
 }
 nodes [
  n1 = CameraVectorWS()
  n2 = CameraPositionWS()
  n3 = MaterialFunctionCall!(
   UVs=Subtract(A=Multiply(A=@n1) {constB=0.18}, B=Multiply(A=@n2) {constB=0.0007})
   Slices Texture=TextureObject() {samplerType=SAMPLERTYPE_LinearColor, texture=/Game/Environments/Helios/TechArt/LowAltitudeClouds/T_LowAltitudeClouds_N}
   Atlas Size=Constant() {r=8}
   Atlas Padding=ScalarParameter() {defaultValue=0.9375, parameterName=AtlasPadding, primitiveDataIndex=8}
  ) {materialFunction=/Game/Environments/Helios/TechArt/LowAltitudeClouds/MF_3DVolumeFrom2DSlices}
  n4 = OneMinus(None=ComponentMask(None=@n3.RGB) {b=true})
  n5 = Multiply(A=@n4, B=@n4)
  n6 = SkyAtmosphereLightDirection()
  n7 = Multiply(A=ActorPositionWS() {originType=CameraRelative}) {constB=-1}
  n8 = Divide(
   A=Transform(None=@n7) {transformSourceType=TRANSFORMSOURCE_World, transformType=TRANSFORM_Local}
   B=Multiply(A=MaterialFunctionCall() {materialFunction=&ObjectScale}.Scale XYZ) {constB=100}
  )
  n9 = MaterialFunctionCall(
   R=ScalarParameter() {bUseCustomPrimitiveData=true, parameterName=ShadingScaleX, primitiveDataIndex=12}
   G=ScalarParameter() {bUseCustomPrimitiveData=true, parameterName=ShadingScaleY, primitiveDataIndex=13}
   B=ScalarParameter() {bUseCustomPrimitiveData=true, parameterName=ShadingScaleZ, primitiveDataIndex=14}
  ) {materialFunction=&AppendMany}
  n10 = ScalarParameter() {bUseCustomPrimitiveData=true, parameterName=ShadingOffsetX, primitiveDataIndex=15}
  n11 = ScalarParameter() {bUseCustomPrimitiveData=true, parameterName=ShadingOffsetY, primitiveDataIndex=16}
  n12 = ScalarParameter() {bUseCustomPrimitiveData=true, parameterName=ShadingOffsetZ, primitiveDataIndex=17}
  n13 = Transform(None=Subtract(
    A=Divide(A=@n8, B=@n9.RGB) {constB=160}
    B=MaterialFunctionCall(R=@n10, G=@n11, B=@n12) {materialFunction=&AppendMany}.RGB
   )) {transformSourceType=TRANSFORMSOURCE_Local}
  n14 = Add(A=Multiply(A=Add(A=@n5) {constB=-0.5}) {constB=2}, B=VertexInterpolator(VS=DotProduct(A=@n6, B=@n13)).PS)
  n15 = Max(A=Add(A=Multiply(A=@n14) {constB=0.5}) {constB=0.7}) {constB=0}
  n16 = CameraVectorWS()
  n17 = SkyAtmosphereLightDirection()
  n18 = Saturate(None=Multiply(A=DotProduct(A=@n16, B=@n17)) {constB=-1})
  n19 = Multiply(A=Multiply(A=@n18, B=@n18)) {constB=0.4}
  Multiply(
   A=MaterialFunctionCall!(DotMask=Add(A=@n15, B=@n19), Halo Mask) {materialFunction=&/MF_LowAltitudeClouds_V2_Colors}.Result
   B=MaterialFunctionCall(
    R=ScalarParameter() {bUseCustomPrimitiveData=true, parameterName=ColorMultiplierX, primitiveDataIndex=19}
    G=ScalarParameter() {bUseCustomPrimitiveData=true, parameterName=ColorMultiplierY, primitiveDataIndex=20}
    B=ScalarParameter() {bUseCustomPrimitiveData=true, parameterName=ColorMultiplierZ, primitiveDataIndex=21}
   ) {materialFunction=&AppendMany}.RGB
  )
  n20 = MaterialFunctionCall(
   R=ScalarParameter() {bUseCustomPrimitiveData=true, parameterName=AxisXX, primitiveDataIndex=3}
   G=ScalarParameter() {bUseCustomPrimitiveData=true, parameterName=AxisXY, primitiveDataIndex=4}
   B=ScalarParameter() {bUseCustomPrimitiveData=true, parameterName=AxisXZ, primitiveDataIndex=5}
  ) {materialFunction=&AppendMany}
  n21 = MaterialFunctionCall(
   R=ScalarParameter() {bUseCustomPrimitiveData=true, parameterName=ScaleX}
   G=ScalarParameter() {bUseCustomPrimitiveData=true, parameterName=ScaleY, primitiveDataIndex=1}
   B=ScalarParameter() {bUseCustomPrimitiveData=true, parameterName=ScaleZ, primitiveDataIndex=2}
  ) {materialFunction=&AppendMany}
  n22 = Subtract(A=Abs(None=DotProduct(A=@n7, B=@n20.RGB)), B=ComponentMask(None=@n21.RGB) {r=true})
  n23 = MaterialFunctionCall(
   R=ScalarParameter() {bUseCustomPrimitiveData=true, parameterName=AxisYX, primitiveDataIndex=6}
   G=ScalarParameter() {bUseCustomPrimitiveData=true, parameterName=AxisYY, primitiveDataIndex=7}
   B=ScalarParameter() {bUseCustomPrimitiveData=true, parameterName=AxisYZ, primitiveDataIndex=8}
  ) {materialFunction=&AppendMany}
  n24 = Subtract(A=Abs(None=DotProduct(A=@n7, B=@n23.RGB)), B=ComponentMask(None=@n21.RGB) {g=true})
  n25 = ScalarParameter() {bUseCustomPrimitiveData=true, parameterName=AxisZX, primitiveDataIndex=9}
  n26 = ScalarParameter() {bUseCustomPrimitiveData=true, parameterName=AxisZY, primitiveDataIndex=10}
  n27 = ScalarParameter() {bUseCustomPrimitiveData=true, parameterName=AxisZZ, primitiveDataIndex=11}
  n28 = Abs(None=DotProduct(A=@n7, B=MaterialFunctionCall(R=@n25, G=@n26, B=@n27) {materialFunction=&AppendMany}.RGB))
  n29 = ComponentMask(None=@n21.RGB) {b=true}
  n30 = Min(A=Min(A=OneMinus(None=@n22), B=OneMinus(None=@n24)), B=OneMinus(None=Subtract(A=@n28, B=@n29)))
  n31 = Ceil(None=@n30)
  Multiply(A=OneMinus(None=Saturate(None=@n31)), B=Constant3Vector() {constant=<rgba 0 0 -500000 0>})
  n32 = ComponentMask(None=@n3.A) {b=true}
  n33 = Divide(A=@n30, B=ScalarParameter() {bUseCustomPrimitiveData=true, defaultValue=1.8, parameterName=Fade, primitiveDataIndex=18})
  Multiply(A=Multiply(A=@n32, B=@n32), B=VertexInterpolator(VS=Saturate(None=@n33)).PS)
  Multiply(A=@n15, B=@n15)
 ]
}
