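# Unlit translucent material for wispy low-altitude clouds, built from mesh cards driven by custom primitive data.
# Vertex offsets scatter cards pseudo-randomly from local Z, orient them toward the camera, bend them into an arc and collapse them nearby.
# Scrolling cloud textures, eroded radial masks, near fade and depth contact fade shape opacity.
# Color combines sun-direction shading from spherical normals, time-of-day emissive tints and a background cubemap, mostly evaluated per vertex.
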Material /Game/Environments/Helios/TechArt/LowAltitudeClouds/V2/M_LowAltitudeClouds_V2_Whisps
graph {
 properties {
  bAllowFrontLayerTranslucency=false
  bTangentSpaceNormal=false
  bUseLightmapDirectionality=false
  bUsedWithInstancedStaticMeshes=true
  blendMode=BLEND_Translucent
  pixelDepthOffsetMode=PDOM_Legacy
  shadingModel=MSM_Unlit
  translucencyPass=MTP_BeforeDOF
  usedShadingModels=Unlit
 }
 nodes [
  n1 = Multiply(A=ScalarParameter() {bUseCustomPrimitiveData=true, defaultValue=800, parameterName=Alpha Scale, primitiveDataIndex=21}) {constB=10}
  n2 = Divide(A=NamedRerouteUsage(), B=@n1) {constB=400}
  n3 = Add(A=Multiply(A=@n2) {constB=0.5}, B=NamedRerouteUsage())
  n4 = Multiply(
   A=Multiply(A=Time(), B=Constant2Vector() {g=-0.0365, r=0.025}) {constB=0.02}
   B=ScalarParameter() {bUseCustomPrimitiveData=true, defaultValue=1, parameterName=Scroll Speed, primitiveDataIndex=14}
  )
  n5 = TextureSample(UVs=VertexInterpolator(VS=Add(A=@n3, B=@n4)).PS) {samplerType=SAMPLERTYPE_LinearGrayscale, texture=/Game/Environments/Helios/TechArt/LowAltitudeClouds/T_LowAltitudeClouds_A}
  n6 = Constant2Vector() {g=1, r=0.25}
  n7 = Time()
  n8 = Constant2Vector() {g=0.0165, r=-0.015}
  n9 = Add(
   A=Add(A=Multiply(A=@n2, B=@n6), B=NamedRerouteUsage())
   B=Multiply(
    A=Multiply(A=@n7, B=@n8) {constB=0.02}
    B=ScalarParameter() {bUseCustomPrimitiveData=true, defaultValue=1, parameterName=Scroll Speed, primitiveDataIndex=14}
   )
  )
  n10 = MaterialFunctionCall() {materialFunction=&ObjectScale}
  n11 = NamedRerouteUsage()
  n12 = Multiply(
   A=Add(A=Transform(None=@n11) {transformSourceType=TRANSFORMSOURCE_World, transformType=TRANSFORM_Local}, B=LocalPosition().XYZ)
   B=MaterialFunctionCall() {materialFunction=&ObjectScale}.Scale XYZ
  )
  n13 = ComponentMask(None=@n12) {g=true}
  n14 = ComponentMask(None=@n12) {r=true}
  n15 = Divide(A=If(A=@n10.Scale X, B=@n10.Scale X, "A > B"=ComponentMask(None=@n12) {g=true, r=true}, "A < B"=AppendVector(A=@n13, B=@n14)), B=@n1) {constB=400}
  n16 = Constant2Vector() {g=1, r=0.5}
  n17 = Time()
  n18 = Constant2Vector() {g=0.0185, r=-0.018}
  n19 = Add(
   A=Add(A=Multiply(A=@n15, B=@n16), B=NamedRerouteUsage())
   B=Multiply(
    A=Multiply(A=@n17, B=@n18) {constB=0.02}
    B=ScalarParameter() {bUseCustomPrimitiveData=true, defaultValue=1, parameterName=Scroll Speed, primitiveDataIndex=14}
   )
  )
  n20 = ActorPositionWS() {originType=CameraRelative}
  n21 = Abs(None=ComponentMask(None=Normalize(VectorInput=@n20)) {b=true})
  n22 = LinearInterpolate(
   A=TextureSample(UVs=VertexInterpolator(VS=@n9).PS) {samplerType=SAMPLERTYPE_LinearGrayscale, texture=&/T_CloudStrip}.R
   B=TextureSample(UVs=VertexInterpolator(VS=@n19).PS) {samplerType=SAMPLERTYPE_LinearGrayscale, texture=&T_CloudStrip}.R
   Alpha=Multiply(A=@n21, B=@n21)
  )
  NamedRerouteDeclaration(None=Multiply(A=Add(A=@n5.R, B=@n22), B=ScalarParameter() {bUseCustomPrimitiveData=true, defaultValue=0.5, parameterName=Opacity, primitiveDataIndex=20})) {name=Alpha, nodeColor=<rgba 1 0.6797 0 1>}
  n23 = NamedRerouteUsage()
  n24 = SkyAtmosphereLightDirection()
  n25 = Add(A=Multiply(A=DotProduct(A=@n23, B=@n24)) {constB=0.5}) {constB=0.5}
  n26 = Multiply(A=Normalize(VectorInput=WorldPosition() {worldPositionShaderOffset=WPT_CameraRelative}.XYZ)) {constB=-1}
  n27 = SkyAtmosphereLightDirection()
  n28 = Saturate(None=Multiply(A=DotProduct(A=@n26, B=@n27)) {constB=-1})
  n29 = Multiply(A=@n28, B=@n28)
  n30 = Multiply(A=@n29, B=@n29)
  n31 = LinearInterpolate(Alpha=NamedRerouteUsage()) {constA=0.1}
  n32 = MaterialFunctionCall!(
   DotMask=Add(A=Power(Base=@n25) {constExponent=1.5}, B=Multiply(A=@n29) {constB=0.4})
   Halo Mask=Multiply(A=Multiply(A=@n30, B=@n31) {constA=4, constB=4}) {constB=0.5}
  ) {materialFunction=&/MF_LowAltitudeClouds_V2_Colors}
  n33 = MaterialFunctionCall(
   R=ScalarParameter() {bUseCustomPrimitiveData=true, parameterName=ColorMultiplierX, primitiveDataIndex=23}
   G=ScalarParameter() {bUseCustomPrimitiveData=true, parameterName=ColorMultiplierY, primitiveDataIndex=24}
   B=ScalarParameter() {bUseCustomPrimitiveData=true, parameterName=ColorMultiplierZ, primitiveDataIndex=25}
  ) {materialFunction=&AppendMany}
  n34 = MaterialFunctionCall(
   R=ScalarParameter() {bUseCustomPrimitiveData=true, parameterName=EmissiveMorningX, primitiveDataIndex=8}
   G=ScalarParameter() {bUseCustomPrimitiveData=true, parameterName=EmissiveMorningY, primitiveDataIndex=9}
   B=ScalarParameter() {bUseCustomPrimitiveData=true, parameterName=EmissiveMorningZ, primitiveDataIndex=10}
  ) {materialFunction=&AppendMany}
  n35 = MaterialFunctionCall(
   R=ScalarParameter() {bUseCustomPrimitiveData=true, parameterName=EmissiveDayX, primitiveDataIndex=11}
   G=ScalarParameter() {bUseCustomPrimitiveData=true, parameterName=EmissiveDayY, primitiveDataIndex=12}
   B=ScalarParameter() {bUseCustomPrimitiveData=true, parameterName=EmissiveDayZ, primitiveDataIndex=13}
  ) {materialFunction=&AppendMany}
  n36 = MaterialFunctionCall(
   R=ScalarParameter() {bUseCustomPrimitiveData=true, parameterName=EmissiveEveningX, primitiveDataIndex=26}
   G=ScalarParameter() {bUseCustomPrimitiveData=true, parameterName=EmissiveEveningY, primitiveDataIndex=27}
   B=ScalarParameter() {bUseCustomPrimitiveData=true, parameterName=EmissiveEveningZ, primitiveDataIndex=28}
  ) {materialFunction=&AppendMany}
  n37 = MaterialFunctionCall(
   R=ScalarParameter() {bUseCustomPrimitiveData=true, parameterName=EmissiveNightX, primitiveDataIndex=29}
   G=ScalarParameter() {bUseCustomPrimitiveData=true, parameterName=EmissiveNightY, primitiveDataIndex=30}
   B=ScalarParameter() {bUseCustomPrimitiveData=true, parameterName=EmissiveNightZ, primitiveDataIndex=31}
  ) {materialFunction=&AppendMany}
  n38 = Constant() {r=1}
  n39 = Constant() {r=1.1}
  n40 = Multiply(
   A=Add(
    A=Multiply(A=@n32.Result, B=@n33.RGB)
    B=MaterialFunctionCall(Color_Morning=@n34.RGB, Color_Day=@n35.RGB, Color_Evening=@n36.RGB, Color_Night=@n37.RGB) {materialFunction=&Color_Time_of_Day}.Result
   )
   B=QualitySwitch(Default=@n38, Low=@n39, High=@n38, Medium=@n39, Epic=@n38)
  ) {constB=1.2}
  n41 = TextureSample(UVs=Multiply(A=CameraVectorWS()) {constB=-1}) {
   samplerType=SAMPLERTYPE_LinearColor
   texture=&/T_LowAltitudeClouds_BackgroundCubemap_Tex
  }
  n42 = Multiply(
   A=@n41.RGB
   B=ComponentMask(None=CollectionParameter() {collection=&/MPC_LowAltitudeClouds_V2, parameterName=ColorLight}) {b=true, g=true, r=true}
  )
  NamedRerouteDeclaration(None=LinearInterpolate(A=VertexInterpolator(VS=@n40).PS, B=Multiply(A=@n42) {constB=2.5}) {constAlpha=0.4}) {name=Color, nodeColor=<rgba 0 1 0.6953 1>}
  n43 = Divide(
   A=Transform(None=NamedRerouteUsage()) {transformSourceType=TRANSFORMSOURCE_World, transformType=TRANSFORM_Instance}
   B=Multiply(A=MaterialFunctionCall() {materialFunction=&ObjectScale}.Scale XYZ) {constB=50}
  )
  n44 = Normalize(VectorInput=WorldPosition() {worldPositionShaderOffset=WPT_CameraRelative}.XYZ)
  NamedRerouteDeclaration(None=Add(A=Transform(None=@n43) {transformSourceType=TRANSFORMSOURCE_Instance}, B=Multiply(A=@n44) {constB=-1})) {name=Spherical Normals, nodeColor=<rgba 1 0.8438 0 1>}
  n45 = Add(A=TextureCoordinate()) {constB=-0.5}
  MaterialExpressionMultiply_98 = Multiply(A=Multiply(A=@n45, B=@n45)) {constB=4.5}
  n46 = Convert(None=@MaterialExpressionMultiply_98) {
   nodeName=Break Float2
   convertInputs [{
     type=Vector2
     expressionInput {
      expression=&self:MaterialExpressionMultiply_98
      inputName=None
      mask=0
      maskA=0
      maskB=0
      maskG=0
      maskR=0
      outputIndex=0
     }
    }]
   convertOutputs [{type=Scalar}, {type=Scalar}]
  }
  n47 = Add(A=@n46, B=@n46)
  n48 = OneMinus(None=NamedRerouteDeclaration(None=Saturate(None=@n47)) {name=Radial Mask, nodeColor=<rgba 0 0.1562 1 1>})
  n49 = Multiply(A=Multiply(A=@n48, B=@n48), B=NamedRerouteUsage())
  n50 = OneMinus(None=Power(
    Base=Saturate(None=@n49)
    Exp=ScalarParameter() {bUseCustomPrimitiveData=true, defaultValue=1.8, parameterName=Erosion, primitiveDataIndex=19}
   ) {constExponent=1.5})
  n51 = OneMinus(None=Multiply(A=@n50, B=@n50) {constB=1.8})
  n52 = VertexInterpolator(VS=NamedRerouteUsage())
  n53 = Divide(
   A=Subtract(A=SceneDepth().R, B=PixelDepth().R)
   B=ScalarParameter() {bUseCustomPrimitiveData=true, defaultValue=1000, parameterName=Contact Fade, primitiveDataIndex=16}
  )
  NamedRerouteDeclaration(None=Multiply(A=Multiply(A=@n51, B=@n52.PS), B=Saturate(None=@n53))) {name=Opacity, nodeColor=<rgba 0 0.4141 1 1>}
  n54 = ScalarParameter() {bUseCustomPrimitiveData=true, defaultValue=15000, group=NearFade, parameterName=Near Fade Start, primitiveDataIndex=18}
  n55 = Distance(A=NamedRerouteUsage(), B=Constant())
  n56 = Distance(A=ActorPositionWS() {originType=CameraRelative}, B=Constant())
  n57 = Saturate(None=InverseLinearInterpolate(
    A=ScalarParameter() {bUseCustomPrimitiveData=true, defaultValue=5000, group=NearFade, parameterName=Near Fade End, primitiveDataIndex=17}
    B=@n54
    Value=Min(A=@n55, B=@n56)
   ) {constA=2000, constB=5000})
  n58 = Abs(None=LocalPosition().Z)
  n59 = Multiply(A=@n58) {constB=0.6599}
  n60 = ScalarParameter() {bUseCustomPrimitiveData=true, defaultValue=1, parameterName=Seed}
  n61 = Multiply(A=@n58) {constB=0.1369}
  n62 = Multiply(A=@n58) {constB=0.4688}
  n63 = MaterialFunctionCall(R=Frac(None=Multiply(A=@n59, B=@n60)), G=Frac(None=Multiply(A=@n61, B=@n60)), B=Frac(None=Multiply(A=@n62, B=@n60))) {materialFunction=&AppendMany}
  n64 = NamedRerouteDeclaration(None=Multiply(A=Add(A=@n63.RGB) {constB=-0.5}) {constB=2}) {name=Random Local Positions, nodeColor=<rgba 0.3125 1 0 1>}
  n65 = ComponentMask(None=@n64) {g=true, r=true}
  n66 = LinearInterpolate(
   A=ComponentMask(None=@n64) {b=true}
   Alpha=ScalarParameter() {bUseCustomPrimitiveData=true, parameterName=Trapezoidal Shape, primitiveDataIndex=7}
  ) {constB=-1}
  n67 = OneMinus(None=NamedRerouteUsage())
  n68 = Multiply(A=Clamp(None=AppendVector(A=@n65, B=@n66), Min=Multiply(A=@n67) {constB=-1}, Max=@n67) {maxDefault=20, minDefault=-20}) {constB=100}
  n69 = Multiply(A=LocalPosition().XYZ) {constB=-1}
  n70 = Constant2Vector()
  n71 = LocalPosition()
  n72 = ComponentMask(None=@n71.XY) {r=true}
  n73 = Multiply(A=Abs(None=@n72)) {constB=-0.4}
  n74 = NamedRerouteUsage()
  n75 = Normalize(VectorInput=CrossProduct(A=Constant3Vector() {constant=<rgba 0 0 1 0>}, B=@n74))
  n76 = Multiply(A=@n75, B=@n72)
  n77 = Multiply(A=CrossProduct(A=@n75, B=@n74), B=ComponentMask(None=@n71.XY) {g=true})
  n78 = MaterialFunctionCall(Float3=Abs(None=NamedRerouteUsage())) {materialFunction=&BreakOutFloat3Components}
  n79 = LinearInterpolate(
   A=MaterialFunctionCall(
    R=ScalarParameter() {bUseCustomPrimitiveData=true, parameterName=SizeMaxX, primitiveDataIndex=1}
    G=ScalarParameter() {bUseCustomPrimitiveData=true, parameterName=SizeMaxY, primitiveDataIndex=2}
    B=ScalarParameter() {bUseCustomPrimitiveData=true, parameterName=SizeMaxZ, primitiveDataIndex=3}
   ) {materialFunction=&AppendMany}.RGB
   B=MaterialFunctionCall(
    R=ScalarParameter() {bUseCustomPrimitiveData=true, parameterName=SizeMinX, primitiveDataIndex=4}
    G=ScalarParameter() {bUseCustomPrimitiveData=true, parameterName=SizeMinY, primitiveDataIndex=5}
    B=ScalarParameter() {bUseCustomPrimitiveData=true, parameterName=SizeMinZ, primitiveDataIndex=6}
   ) {materialFunction=&AppendMany}.RGB
   Alpha=Max(A=Max(A=@n78.R, B=@n78.R), B=@n78.B)
  ) {constA=1}
  n80 = Multiply(
   A=Add(A=AppendVector(A=@n70, B=@n73), B=Add(A=@n76, B=@n77))
   B=NamedRerouteDeclaration(None=Min(A=@n79)) {name=Cards Sizes, nodeColor=<rgba 1 0.4922 0 1>}
  )
  n81 = Transform(None=Add(A=Add(A=@n68, B=@n69), B=NamedRerouteDeclaration(None=@n80) {name=Cross Product Expand, nodeColor=<rgba 0 0.7891 1 1>}) {constB=0}) {transformSourceType=TRANSFORMSOURCE_Instance}
  n82 = Subtract(A=Add(A=@n81, B=WorldPosition() {worldPositionShaderOffset=WPT_CameraRelative}.XYZ), B=ActorPositionWS() {originType=CameraRelative})
  n83 = Multiply(A=Transform(None=NamedRerouteDeclaration(None=@n82) {name=Origin Position, nodeColor=<rgba 1 0.1641 0 1>}) {transformSourceType=TRANSFORMSOURCE_World, transformType=TRANSFORM_Local}) {constB=0.01}
  n84 = Abs(None=ComponentMask(None=Abs(None=@n83)) {g=true})
  n85 = ScalarParameter() {bUseCustomPrimitiveData=true, parameterName=Arc Shape, primitiveDataIndex=15}
  n86 = Constant()
  n87 = Multiply(
   A=Add(A=Multiply(A=@n84, B=@n84)) {constB=-0.5}
   B=Transform(None=MaterialFunctionCall(R=@n85, G=@n86, B=@n86) {materialFunction=&AppendMany}.RGB) {transformSourceType=TRANSFORMSOURCE_Local}
  )
  n88 = WorldPosition() {worldPositionShaderOffset=WPT_CameraRelative}
  NamedRerouteDeclaration(None=If(
    A=NamedRerouteDeclaration(None=@n57) {name=Distance Mask, nodeColor=<rgba 0.3594 0 1 1>}
    "A > B"=Add(A=@n81, B=@n87)
    "A < B"=Multiply(A=@n88.XYZ) {constB=-1}
   ) {constB=0.01}) {name=WPO, nodeColor=<rgba 0 1 0.1094 1>}
  n89 = ActorPositionWS() {originType=CameraRelative}
  n90 = Transform(None=@n68) {transformSourceType=TRANSFORMSOURCE_Local}
  n91 = Normalize(VectorInput=NamedRerouteDeclaration(None=Add(A=@n89, B=@n90)) {name=Random World Positions, nodeColor=<rgba 0.3828 0 1 1>})
  NamedRerouteDeclaration(None=Multiply(
    A=Transform(None=@n91) {transformSourceType=TRANSFORMSOURCE_World, transformType=TRANSFORM_Local}
    B=MaterialFunctionCall() {materialFunction=&ObjectScale}.Scale XYZ
   )) {name=Card vector, nodeColor=<rgba 1 0 0.1172 1>}
  n92 = Add(A=WorldPosition() {worldPositionShaderOffset=WPT_CameraRelative}.XYZ, B=@n81)
  n93 = NamedRerouteUsage()
  n94 = Constant2Vector() {g=1, r=1}
  n95 = ScalarParameter() {bUseCustomPrimitiveData=true, defaultValue=1.7311, parameterName=Scale Z, primitiveDataIndex=22}
  n96 = Divide(A=Multiply(A=Subtract(A=@n92, B=@n93), B=AppendVector(A=@n94, B=@n95)), B=LinearInterpolate(Alpha=NamedRerouteUsage()) {constA=0.3})
  NamedRerouteDeclaration(None=ComponentMask(None=Transform(None=@n96) {transformSourceType=TRANSFORMSOURCE_World, transformType=TRANSFORM_View}) {g=true, r=true}) {name=Alpha Coords, nodeColor=<rgba 0.6172 1 0 1>}
  NamedRerouteDeclaration(None=@n63.RG) {name=Random UV Offset, nodeColor=<rgba 0.1953 1 0 1>}
  NamedRerouteUsage()
  NamedRerouteUsage()
  NamedRerouteUsage()
 ]
}
